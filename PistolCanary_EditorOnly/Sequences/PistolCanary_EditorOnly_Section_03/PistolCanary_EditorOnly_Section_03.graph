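# Editor-only Level Sequence for Section 03 of PistolCanary that plays an animatic video with matching audio.
# A spawn-tracked static mesh plane is held at a fixed scaled transform with an attach track.
# A material track swaps its mesh slot to the Section 03 v6 media texture material.
# Top-level media and audio tracks start on the same frame, with audio running slightly longer than the video.
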LevelSequence /PistolCanary_EditorOnly/Sequences/PistolCanary_EditorOnly_Section_03/PistolCanary_EditorOnly_Section_03
bindings [
 SM_Animatic_Plane {
  id=#1
  possessed_class=FortStaticMeshActor
  tracks [
   {type=MovieSceneSpawnTrack, sections [{type=MovieSceneSpawnSection, channels [None {type=MovieSceneScriptingBoolChannel, default=true}]}]}
   {
    type=MovieScene3DTransformTrack
    sections [{
      type=MovieScene3DTransformSection
      channels [
       Location.X {type=MovieSceneScriptingDoubleChannel, default=20.312}
       Location.Y {type=MovieSceneScriptingDoubleChannel, default=14.0101}
       Location.Z {type=MovieSceneScriptingDoubleChannel, default=6.112}
       Rotation.X {type=MovieSceneScriptingDoubleChannel, default=90}
       Rotation.Y {type=MovieSceneScriptingDoubleChannel, default=0}
       Rotation.Z {type=MovieSceneScriptingDoubleChannel, default=90}
       Scale.X {type=MovieSceneScriptingDoubleChannel, default=0.11}
       Scale.Y {type=MovieSceneScriptingDoubleChannel, default=0.054}
       Scale.Z {type=MovieSceneScriptingDoubleChannel, default=0.096}
      ]
     }]
   }
   {type=MovieScene3DAttachTrack, sections [{type=MovieScene3DAttachSection}]}
  ]
 }
 StaticMeshComponent0 {
  id=#2
  parent_id=#1
  possessed_class=StaticMeshComponent
  tracks [{
    type=MovieScenePrimitiveMaterialTrack
    display_name=Material Slot: lambert1
    sections [{
      type=MovieScenePrimitiveMaterialSection
      channels [None {
        type=MovieSceneScriptingObjectPathChannel
        default="<Object '/PistolCanary_EditorOnly/Layout/Media/Animatic/Section_03/Section_03_v6/PistolCanary_EditorOnly_Section_03_v6_Media_Texture_Mat' (0x852feb42a00) Class 'Material'>"
       }]
     }]
   }]
 }
]
display_rate {denominator=1, numerator=60}
tick_resolution {denominator=1, numerator=24000}
tracks [
 {
  type=MovieSceneAudioTrack
  display_name=PistolCanary_EditorOnly - Audio Section 03
  sections [{
    type=MovieSceneAudioSection
    end_frame=7637
    start_frame=217
    channels [
     Volume {type=MovieSceneScriptingFloatChannel, default=1}
     Pitch {type=MovieSceneScriptingFloatChannel, default=1}
     AttachActor {type=MovieSceneScriptingActorReferenceChannel, default="<Struct 'MovieSceneObjectBindingID' (0x852ff22c690) {}>"}
    ]
   }]
 }
 {
  type=MovieSceneMediaTrack
  display_name=PistolCanary_EditorOnly - Media Section 03
  sections [{type=MovieSceneMediaSection, end_frame=7597, start_frame=217, channels [CanPlayerBeOpen {type=MovieSceneScriptingBoolChannel, default=true}]}]
 }
]
